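; Child Blueprint of the Figment Jackal vehicle that keeps riders from clipping into building actors when boarding.
; IsSeatEnabled box-traces at the Driver bone and disables the seat if it hits a building actor, otherwise deferring to the parent.
; On pawn entry, a one-shot 0.3-second timer runs ClippingCheck.
; It capsule-traces around the driver and between points blended from wheel bones and foot sockets, destroying any damageable building actors hit.
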(event EventBeginPlay
  (|CallParentFunction) ; This node is disabled and will not be called.\nDrag off pins to build functionality.
)

(event Collision|EventActorBeginOverlap
  (params (OtherActor :type Actor%20Object%20Reference)))

(event EventTick
  (params (DeltaSeconds :type Float%20%28single-precision%29)))

(event EventOnPawnEnterVehicle
  (params (PlayerPawn :type Fort%20Player%20Pawn%20Object%20Reference) (SeatIdx :type Integer))
  (|CallParentFunction PlayerPawn SeatIdx)
  (Utilities|Time|SetTimerbyEvent (delegate Custom|ClippingCheck) 0.3))

(event Custom|ClippingCheck
  (bind _array (Utilities|Array|MakeArray "ObjectTypeQuery6" "ObjectTypeQuery1"))
  (bind _outhits (Collision|MultiCapsuleTraceForObjects (- (Physics|GetCenterOfMass (Variables|/Script/Engine.Character|GetCapsuleComponent (Vehicle|GetDriver)) "None") "0.000000,0.000000,20.000000") (- (Physics|GetCenterOfMass (Variables|/Script/Engine.Character|GetCapsuleComponent (Vehicle|GetDriver)) "None") "0.000000,0.000000,20.000000") 2e+01 75.0 _array))
  (for _array_element _outhits
    (if (and (Variables|/Script/Engine.Actor|GetbCanBeDamaged (.HitActor (Collision|BreakHitResult _array_element))) (Utilities|ClassIsChildOf (Utilities|GetClass (.HitActor (Collision|BreakHitResult _array_element))) "/Script/FortniteGame.BuildingActor"))
      (Actor|DestroyActor (.HitActor (Collision|BreakHitResult _array_element)))))
  (bind _outhits_2 (Collision|MultiCapsuleTraceForObjects (Math|Vector|Lerp(Vector) (.location (Transformation|GetBoneTransform (Variables|Vehicle|GetSkeletalMesh) "FrontWheels" "RTS_World")) (.location (Transformation|GetBoneTransform (Variables|Vehicle|GetSkeletalMesh) "FootLeftSocket" "RTS_World")) 0.3) (Math|Vector|Lerp(Vector) (.location (Transformation|GetBoneTransform (Variables|Vehicle|GetSkeletalMesh) "RearWheels" "RTS_World")) (.location (Transformation|GetBoneTransform (Variables|Vehicle|GetSkeletalMesh) "FootRightSocket" "RTS_World")) 0.3) 2e+01 1e+01 _array))
  (for _array_element_2 _outhits_2
    (if (and (Variables|/Script/Engine.Actor|GetbCanBeDamaged (.HitActor (Collision|BreakHitResult _array_element_2))) (Utilities|ClassIsChildOf (Utilities|GetClass (.HitActor (Collision|BreakHitResult _array_element_2))) "/Script/FortniteGame.BuildingActor"))
      (Actor|DestroyActor (.HitActor (Collision|BreakHitResult _array_element_2))))))

(fn IsSeatEnabled
  (params (SeatIdx :type Integer))
  (Utilities|FlowControl|Sequence
    (:then_0
      ;[1 1 1 0.5] Checks to make sure there's no props that may be blocking entry into the sheet that may otherwise lack player collission (FORT-1109937)
      (bind _outhits (Collision|MultiBoxTraceForObjects (- (.location (Transformation|GetBoneTransform (Variables|Vehicle|GetSkeletalMesh) "Driver" "RTS_World")) "0.000000,0.000000,20.000000") (- (.location (Transformation|GetBoneTransform (Variables|Vehicle|GetSkeletalMesh) "Driver" "RTS_World")) "0.000000,0.000000,20.000000") "25.000000,25.000000,40.000000" (.rotation (Transformation|GetBoneTransform (Variables|Vehicle|GetSkeletalMesh) "Driver" "RTS_World")) (Utilities|Array|MakeArray "ObjectTypeQuery6" "ObjectTypeQuery1")))
      (for _array_element _outhits
        (if (Utilities|ClassIsChildOf (Utilities|GetClass (.HitActor (Collision|BreakHitResult _array_element))) "/Script/FortniteGame.BuildingActor")
          (return false))))
    (:then_1
      (bind _returnvalue (|CallParentFunction SeatIdx))
      (return _returnvalue))))

(fn ConstructionScript
  (|CallParentFunction))
